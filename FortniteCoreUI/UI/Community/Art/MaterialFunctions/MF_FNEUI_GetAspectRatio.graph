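# Material function for UI materials that outputs a widget's aspect ratios, exposed to the material library.
# It splits Pixel Size from GetUserInterfaceUV into X and Y, which keeps the aspect ratio correct.
# Divides yield X/Y and Y/X scalars, each paired with 1 as horizontal and vertical Vector2 ratios.
# Min and Max nodes give smallest and largest ratios, and MF_FNEUI_BooleanOperations flags whether width exceeds height.
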MaterialFunction /FortniteCoreUI/UI/Community/Art/MaterialFunctions/MF_FNEUI_GetAspectRatio
graph {
 properties {
  bExposeToLibrary=true
  description="This MF is outputing the ratio of your widget base on the higher segment (width or height). It is using the Pixel Size output from the GetUserInterfaceUV MF "
 }
 nodes [
  n1 = NamedRerouteUsage()
  ComponentMask(None=@n1) {r=true}
  ComponentMask(None=@n1) {g=true}
  n2 = NamedRerouteUsage()
  ComponentMask(None=@n2) {r=true}
  ComponentMask(None=@n2) {g=true}
  n3 = NamedRerouteUsage()
  n4 = NamedRerouteUsage()
  n5 = FunctionOutput(None=NamedRerouteDeclaration(None=Divide(A=@n3, B=@n4)) {name=x/y.ratio, nodeColor=<rgba 1 0.0469 0 1>}) {outputName="X/Y Ratio (S)"}
  n6 = Constant() {r=1}
  FunctionOutput(None=NamedRerouteDeclaration(None=AppendVector(A=@n5, B=@n6)) {name=hRatio, nodeColor=<rgba 0 0 0 1>}) {description="Y=1, X= Horizontal Ratio (Return a Vector2)", outputName="H Ratio (V2)", sortPriority=2}
  n7 = NamedRerouteUsage()
  n8 = NamedRerouteUsage()
  n9 = FunctionOutput(None=NamedRerouteDeclaration(None=Divide(A=@n7, B=@n8)) {name=y/x.ratio, nodeColor=<rgba 1 0.0469 0 1>}) {outputName="Y/X Ratio (S)", sortPriority=1}
  FunctionOutput(None=NamedRerouteDeclaration(None=AppendVector(A=@n6, B=@n9)) {name=vRatio, nodeColor=<rgba 0 0 0 1>}) {description="X=1, Y= Vertical Ratio (Return a Vector2) ", outputName="V Ratio (V2)", sortPriority=3}
  n10 = NamedRerouteUsage()
  n11 = NamedRerouteUsage()
  FunctionOutput(None=NamedRerouteDeclaration(None=Max(A=@n10, B=@n11) {desc="Get the highest ratio float value \r\nbetween widget h.size and v.size."}) {name=Ratio of the Current Highest Widget Segment - Float, nodeColor=<rgba 0.382 1 0 1>}) {
   description="Ratio of the Current Highest Widget Segment (return a Float).\r\n\r\nGet the highest ratio float value between widget h.size and v.size.\r\n"
   outputName="Highest Widget Segment Ratio (S)"
   sortPriority=7
  }
  n12 = NamedRerouteUsage()
  n13 = NamedRerouteUsage()
  FunctionOutput(None=NamedRerouteDeclaration(None=Max(A=@n12, B=@n13)) {name=Max Ratio, nodeColor=<rgba 0.4181 1 0.0555 1>}) {description=Set Vertical Ratio or Horizontal ratio depending the highest segment of both., outputName="Max Ratio (V2)", sortPriority=5}
  n14 = NamedRerouteDeclaration(None=MaterialFunctionCall!(A=@n10, B=@n11) {materialFunction=&/MF_FNEUI_BooleanOperations}.IsGreater) {
   name=Widget Width is greater than its Height
   desc="0 = Widget Size Vertical Segment the Higher\r\n1 = Widget Size Horizontal Segment is the Higher"
   nodeColor=<rgba 1 0 0.1402 1>
  }
  FunctionOutput(None=@n14) {
   description="Return 0 or 1.\r\n\r\n0 = Widget Size Vertical Segment the Higher\r\n\r\n1 = Widget Size Horizontal Segment is the Higher\r\n"
   outputName="Widget Width is greater than its Height (S)"
   sortPriority=8
  }
  FunctionOutput(None=NamedRerouteDeclaration(None=Min(A=@n10, B=@n11) {desc="Get the smallest ratio float value \r\nbetween widget h.size and v.size."}) {name=Ratio of the Current Smallest Widget Segment - Float, nodeColor=<rgba 0.8846 1 0 1>}) {
   description="Ratio of the Current Smallest Widget Segment (return a Float).\r\n\r\nGet the smallest ratio float value between widget h.size and v.size.\r\n"
   outputName="Smallest Widget Segment Ratio (S)"
   sortPriority=6
  }
  n15 = NamedRerouteUsage()
  n16 = NamedRerouteUsage()
  FunctionOutput(None=NamedRerouteDeclaration(None=Min(A=@n15, B=@n16)) {name=Min Ratio, nodeColor=<rgba 0.4181 1 0.0555 1>}) {description=Set Vertical Ratio or Horizontal ratio depending the highest segment of both., outputName="Min Ratio (V2)", sortPriority=4}
  n17 = MaterialFunctionCall() {materialFunction=&GetUserInterfaceUV}
  NamedRerouteDeclaration(None=ComponentMask(None=@n17.Pixel Size) {r=true}) {name=pixelSize.x, nodeColor=<rgba 0 0.8125 1 1>}
  NamedRerouteDeclaration(None=ComponentMask(None=@n17.Pixel Size) {g=true}) {name=pixelSize.y, nodeColor=<rgba 1 0.5391 0 1>}
  NamedRerouteDeclaration() {name=Viewport Scale, nodeColor=<rgba 0.5469 1 0 1>}
  Reroute() {desc="To keep Aspect Ratio, you need to use the \"Pixel Size\" from \"GetUserInterface UV\" MF"}
 ]
}
